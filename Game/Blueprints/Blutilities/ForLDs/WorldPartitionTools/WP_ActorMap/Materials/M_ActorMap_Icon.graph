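# Translucent unlit icon material for the World Partition ActorMap level-design tool, drawn without depth testing on camera-aligned meshes.
# It layers a parameterized icon texture over a color-tinted background shape, using eight offset alpha samples to build a dark outline mask.
# World position offset scales each icon by camera distance, view-direction alignment and a global scale from a material parameter collection.
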Material /Game/Blueprints/Blutilities/ForLDs/WorldPartitionTools/WP_ActorMap/Materials/M_ActorMap_Icon
graph {
 properties {
  bDisableDepthTest=true
  bUseTranslucencyVertexFog=false
  bUsedWithInstancedStaticMeshes=true
  bUsedWithNanite=true
  blendMode=BLEND_Translucent
  pixelDepthOffsetMode=PDOM_Legacy
  shadingModel=MSM_Unlit
  usedShadingModels=Unlit
 }
 nodes [
  Add() {constB=0}
  Constant() {r=1}
  n1 = VectorParameter() {parameterName=Color, defaultValue=<rgba 1 0.9096 0.1583 1>}
  n2 = TextureSample() {
   samplerType=SAMPLERTYPE_LinearColor
   texture=/Game/Blueprints/Blutilities/ForLDs/WorldPartitionTools/WP_ActorMap/Textures/GenericDiscoveryIconBG
  }
  n3 = TextureObjectParameter() {parameterName=IconTexture, texture=/Game/Athena/UI/InGame/Creative/PropertyEditor/Icon_Picker/Textures/64x64/T_UI_IconLibrary_waypoint_64}
  n4 = TextureSample(UVs=NamedRerouteUsage(), Tex=@n3) {samplerSource=SSM_Clamp_WorldGroupSettings, texture=/Game/UI/Foundation/Textures/Icons/Athena/T-T-Icon-BR-SM-Athena-Chest-01}
  LinearInterpolate(
   A=LinearInterpolate(A=NamedRerouteUsage(), B=@n1.RGB, Alpha=@n2.R)
   B=LinearInterpolate(A=NamedRerouteUsage(), B=@n4.RGB, Alpha=@n4.RGB)
   Alpha=NamedRerouteUsage()
  )
  MaterialFunctionCall!(
   WorldPosition=NamedRerouteUsage()
   PivotLocation
   WS Normals
   Custom Object Basis 1=VectorParameter() {parameterName=Basis1, defaultValue=<rgba 1 0 0 1>}.RGB
   Custom Object Basis 2=VectorParameter() {parameterName=Basis2, defaultValue=<rgba 0 0 -1 1>}.RGB
   Custom Object Basis 3=VectorParameter() {parameterName=Basis3, defaultValue=<rgba 0 1 0 1>}.RGB
  ) {materialFunction=/Engine/Functions/Engine_MaterialFunctions02/WorldPositionOffset/AlignMeshToTheCamera}
  NamedRerouteDeclaration(None=MaterialFunctionCall(Texture Scale=ScalarParameter() {defaultValue=0.5, parameterName=IconScale}) {materialFunction=&ScaleUVsByCenter}.UVs) {name=ScaledUVs, nodeColor=<rgba 1 0.8906 0 1>}
  n5 = NamedRerouteUsage()
  n6 = MaterialFunctionCall!(OffsetAmount=ScalarParameter() {defaultValue=0.1, parameterName=OffsetAmount}) {materialFunction=/Engine/Functions/Engine_MaterialFunctions01/ImageAdjustment/BlurSampleOffsets}
  n7 = Add(A=@n5, B=@n6."0,1")
  n8 = Add(A=@n5, B=@n6."1,0")
  n9 = Add(A=@n5, B=@n6."0,-1")
  n10 = Add(A=@n5, B=@n6."-1,0")
  n11 = Add(
   A=Add(
    A=TextureSample(UVs=@n7, Tex=@n3) {samplerSource=SSM_Clamp_WorldGroupSettings, texture=&T-T-Icon-BR-SM-Athena-Chest-01}.A
    B=TextureSample(UVs=@n8, Tex=@n3) {samplerSource=SSM_Clamp_WorldGroupSettings, texture=&T-T-Icon-BR-SM-Athena-Chest-01}.A
   )
   B=Add(
    A=TextureSample(UVs=@n9, Tex=@n3) {samplerSource=SSM_Clamp_WorldGroupSettings, texture=&T-T-Icon-BR-SM-Athena-Chest-01}.A
    B=TextureSample(UVs=@n10, Tex=@n3) {samplerSource=SSM_Clamp_WorldGroupSettings, texture=&T-T-Icon-BR-SM-Athena-Chest-01}.A
   )
  )
  n12 = Add(A=@n5, B=@n6."-.5,.5")
  n13 = Add(A=@n5, B=@n6.".5,.5")
  n14 = Add(A=@n5, B=@n6.".5,-.5")
  n15 = Add(A=@n5, B=@n6."-.5,-.5")
  n16 = Add(
   A=Add(
    A=TextureSample(UVs=@n12, Tex=@n3) {samplerSource=SSM_Clamp_WorldGroupSettings, texture=&T-T-Icon-BR-SM-Athena-Chest-01}.A
    B=TextureSample(UVs=@n13, Tex=@n3) {samplerSource=SSM_Clamp_WorldGroupSettings, texture=&T-T-Icon-BR-SM-Athena-Chest-01}.A
   )
   B=Add(
    A=TextureSample(UVs=@n14, Tex=@n3) {samplerSource=SSM_Clamp_WorldGroupSettings, texture=&T-T-Icon-BR-SM-Athena-Chest-01}.A
    B=TextureSample(UVs=@n15, Tex=@n3) {samplerSource=SSM_Clamp_WorldGroupSettings, texture=&T-T-Icon-BR-SM-Athena-Chest-01}.A
   )
  )
  NamedRerouteDeclaration(None=Saturate(None=Add(A=@n11, B=@n16))) {name=OffsetAlpha, nodeColor=<rgba 1 0.8906 0 1>}
  NamedRerouteDeclaration(None=Multiply(A=@n1.RGB) {constB=0.02}) {name=DarkColor, nodeColor=<rgba 0 0.1328 1 1>}
  n17 = WorldPosition()
  n18 = Normalize(VectorInput=Subtract(A=@n17.XYZ, B=ObjectPositionWS()))
  n19 = Subtract(A=ObjectPositionWS(), B=CameraPositionWS())
  n20 = Length(None=@n19) {is=class MaterialExpressionLength}
  n21 = ScalarParameter() {defaultValue=20, parameterName=Dist}
  n22 = ScalarParameter() {defaultValue=-1.5, parameterName=LookAtScale}
  n23 = VectorParameter() {parameterName=CamVector, defaultValue=<rgba 0 0 1 1>}
  n24 = Saturate(None=DotProduct(A=Normalize(VectorInput=@n19), B=Transform(None=@n23.RGB) {transformSourceType=TRANSFORMSOURCE_Camera}))
  n25 = CollectionParameter() {
   collection=&/DontSaveMe_MPC_ActorMap
   parameterName=GlobalWorldIconScale
  }
  n26 = Multiply(A=Multiply(A=Divide(A=@n20, B=@n21) {constB=1}, B=LinearInterpolate(A=@n22, Alpha=@n24) {constA=1}), B=@n25)
  NamedRerouteDeclaration(None=Add(A=@n17.XYZ, B=Multiply(A=@n18, B=@n26))) {name=IconScaleWPO, nodeColor=<rgba 0 0.3906 1 1>}
  Reroute(None=@n2.A)
  Subtract(A=WorldPosition().XYZ, B=ObjectPositionWS())
 ]
}
