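; Level-designer World Partition tool actor that draws an actor map as icon markers, one instanced static mesh component per item type.
; SetDataForMapItems gives each component a dynamic icon material with a fixed-seed random color and a blocking-loaded texture, placing instances at actor-descriptor bounds centers.
; It fills component-to-data and label-to-components maps, which lookup functions query by label or by component and instance index.
; ResetMapActor destroys all instanced mesh components.
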(event EventBeginPlay)

(event Collision|EventActorBeginOverlap
  (params (OtherActor :type Actor%20Object%20Reference)))

(event EventTick
  (params (DeltaSeconds :type Float%20%28single-precision%29)))

(fn GetComponentFromLabel
  (params (label :type String))
  (bind _labelcomponentmap (Variables|Default|GetLabelComponentMap))
  (bind (_value _returnvalue) (Utilities|Map|Find _labelcomponentmap label))
  (if _returnvalue
    (return true (Utilities|Struct|BreakActorMapComponentArray _value))
    (else
      (return false 0))))

(fn GetDataForComponentAndIndex
  (params (Component :type Instanced%20Static%20Mesh%20Component%20Object%20Reference) (Index :type Integer))
  (bind _componentdatamap (Variables|Default|GetComponentDataMap))
  (bind (_value _returnvalue) (Utilities|Map|Find _componentdatamap Component))
  (if _returnvalue
    (LocalVariables|Setl_struct _value)
    (LocalVariables|Setl_instanceData (Utilities|Array|Get(acopy) (Utilities|Struct|BreakActorMapItemStruct (LocalVariables|Getl_struct)) Index))
    (return true (LocalVariables|Getl_struct) (LocalVariables|Getl_instanceData))))

(fn ResetMapActor
  (bind (_arrayindex _arrayelement) (Utilities|Array|ReverseforEachLoop (Actor|GetComponentsByClass "/Script/Engine.InstancedStaticMeshComponent"))
    (:LoopBody
      (Components|DestroyComponent _arrayelement))
    (:Completed
      (Rendering|SetVisibility (Variables|Default|GetDefaultSceneRoot) true true))))

(fn SetDataForMapItems
  (params (ItemData :type Array%20of%20Actor%20Map%20Item%20Struct%20Structures))
  (bind (_perinstancedata_12_64d64295423759f8fa7bd2895df5b243 _itemlabel_10_ff97201248d61a9b8cfea99bee38cad3 _itemicon_14_3d191b4247f7bcc9c30055a9e8a10e01) (Utilities|Struct|BreakActorMapItemStruct (LocalVariables|Getl_curItemData)))
  (bind (_perinstancedata_12_64d64295423759f8fa7bd2895df5b243_2 _itemlabel_10_ff97201248d61a9b8cfea99bee38cad3_2) (Utilities|Struct|BreakActorMapItemStruct (LocalVariables|Getl_curItemData)))
  (bind _l_curitemdata (LocalVariables|Getl_curItemData))
  (bind (_perinstancedata_12_64d64295423759f8fa7bd2895df5b243_3 _itemlabel_10_ff97201248d61a9b8cfea99bee38cad3_3) (Utilities|Struct|BreakActorMapItemStruct _l_curitemdata))
  (bind (_perinstancedata_12_64d64295423759f8fa7bd2895df5b243_4 _itemlabel_10_ff97201248d61a9b8cfea99bee38cad3_4) (Utilities|Struct|BreakActorMapItemStruct (LocalVariables|Getl_curItemData)))
  (bind _labelcomponentmap (Variables|Default|GetLabelComponentMap))
  (bind (_value _returnvalue) (Utilities|Map|Find _labelcomponentmap _itemlabel_10_ff97201248d61a9b8cfea99bee38cad3_3))
  (bind _output_get (Variables|Default|SetCurrentItemData ItemData))
  (LocalVariables|Setl_stream (Math|Random|MakeRandomStream 15))
  (Utilities|Map|Clear (Variables|Default|GetComponentDataMap))
  (Utilities|Map|Clear (Variables|Default|GetLabelComponentMap))
  (for _array_element _output_get
    (bind _output_get_2 (LocalVariables|Setl_curItemData _array_element))
    (LocalVariables|Setl_curInstanceData (Utilities|Struct|BreakActorMapItemStruct _output_get_2))
    (Utilities|FlowControl|Sequence
      (:then_0
        (bind _returnvalue_2 (AddComponent|NODE_AddInstancedStaticMeshComponent-0))
        (LocalVariables|Setl_curComp _returnvalue_2)
        (bind _returnvalue_3 (Rendering|Material|CreateDynamicMaterialInstance "/Game/Blueprints/Blutilities/ForLDs/WorldPartitionTools/WP_ActorMap/Materials/M_ActorMap_Icon.M_ActorMap_Icon" "None" "Transient" :!class /Script/Engine.KismetMaterialLibrary))
        (bind _output_get_3 (LocalVariables|Setl_mat _returnvalue_3))
        (Rendering|Material|SetMaterial (LocalVariables|Getl_curComp) 0 _output_get_3)
        (Rendering|Material|SetVectorParameterValue (LocalVariables|Getl_mat) "Color" (Math|Conversions|ToLinearColor(Vector) (Math|Random|RandomUnitVectorfromStream (LocalVariables|Getl_stream))) :!class /Script/Engine.MaterialInstanceDynamic)
        (bind _returnvalue_4 (Utilities|LoadAssetBlocking _itemicon_14_3d191b4247f7bcc9c30055a9e8a10e01))
        (bind _astexture_2d (Utilities|Casting|CastToTexture2D _returnvalue_4))
        (LocalVariables|Setl_curTex _astexture_2d)
        (Rendering|Material|SetTextureParameterValue (LocalVariables|Getl_mat) "IconTexture" (LocalVariables|Getl_curTex)))
      (:then_1
        (for _array_element_2 (LocalVariables|Getl_curInstanceData)
          (Components|InstancedStaticMesh|AddInstance (LocalVariables|Getl_curComp) (Math|Conversions|ToTransform(Vector) (/ (+ (Utilities|Struct|BreakBox (.Bounds (Utilities|Struct|BreakActorDesc (Utilities|Struct|BreakActorMapPerInstanceStruct _array_element_2)))) (.Max (Utilities|Struct|BreakBox (.Bounds (Utilities|Struct|BreakActorDesc (Utilities|Struct|BreakActorMapPerInstanceStruct _array_element_2)))))) 2.0)) true)))
      (:then_2
        (Utilities|Map|Add (Variables|Default|GetComponentDataMap) (LocalVariables|Getl_curComp) (LocalVariables|Getl_curItemData))
        (if _returnvalue
          (bind _output_get_4 (LocalVariables|Setl_comps (Utilities|Struct|BreakActorMapComponentArray _value)))
          (Utilities|Array|Add _output_get_4 (LocalVariables|Getl_curComp))
          (bind _structout (Utilities|Struct|SetmembersinActorMapComponentArray _value (LocalVariables|Getl_comps)))
          (Utilities|Map|Add (Variables|Default|GetLabelComponentMap) _itemlabel_10_ff97201248d61a9b8cfea99bee38cad3_4 _structout)
          (else
            (Utilities|Map|Add (Variables|Default|GetLabelComponentMap) _itemlabel_10_ff97201248d61a9b8cfea99bee38cad3_2 (Utilities|Struct|MakeActorMapComponentArray (Utilities|Array|MakeArray (LocalVariables|Getl_curComp))))))))))

(fn ConstructionScript)
